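# Niagara system for the DistantEchoPro backpack cosmetic's light, built from a single LightCone_FromHand emitter.
# It burst-spawns one particle that outlives its lifetime and follows a sampled location on the user-supplied skeletal mesh.
# That particle drives mesh and sprite renderers plus two small lights, offset along the sampled rotation's negative Z.
# Alpha and light intensity vary by time of day and are boosted in the front end.
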NiagaraSystem /BRCosmetics/Accessories/FORT_Backpacks/Backpack_M_MED_DistantEchoPro/FX/NS_Backpack_DistantEchoPro_Light
dependencies {
 usedDataInterfaces []
 usedDynamicInputs [&MakeCustomVectorFromBool]
 usedModules [&SystemState]
}
emitters {LightCone_FromHand {
  emitterUpdateScript [
   &EmitterState(
    Inactive Response=::NewEnumerator1
    Life Cycle Mode=::NewEnumerator1
    Loop Behavior=::NewEnumerator1
    Loop Duration Mode=::NewEnumerator1
    UseLoopDelay=false
   )
   &SpawnBurst_Instantaneous(Spawn Count=1, Use Spawn Probability=false)
  ]
  particleSpawnScript [&InitializeParticle(
    Color=<rgba 1 1 1 0.5>
    Lifetime=5
    Mesh Scale Mode=::NewEnumerator3
    Mesh Uniform Scale=0.5
    Sprite Size Mode=::NewEnumerator3
    Uniform Sprite Size=20
    UsePositionOffset=false
    Write Material Random=true
   )]
  particleUpdateScript [
   &ParticleState(Kill Particles When Lifetime Has Elapsed=false, Let Infinitely Lived Particles Die When Emitter Deactivates=true)
   &SkeletalMeshLocation(MaskSpawnGroup=false, Mesh Sampling Type=::NewEnumerator5, Skeletal Mesh=$User.SkeletalMesh, Socket Sampling Mode=::NewEnumerator5)
   &ScaleColor(
    Scale Alpha=&Multiply_Float(
     A=&FN_Select_Float_from_Time_Of_Day(Morning=0.035, Day=0.15, Evening=0.035, Night=0.15)
     B=&FN_FrontEndFloatModifier(Gameplay=1, Front End=4)
    )
    ScaleA=true
    ScaleRGB=true
   )
   &FN_Sprite_Attributes(
    Color=&MultiplyLinearColor(Linear Color=$Particles.Color, Scale Factor=<1 1 1 0.2>)
    Create Camera Offset Attribute=true
    Create Color Attribute=true
    Write to Module Namespace=true
   ) {module=FN_Sprite_Attributes001}
   &FN_Sprite_Attributes(
    Camera Offset=5
    Color=&MultiplyLinearColor(Linear Color=$Particles.Color, Scale Factor=<1 1 1 0.2>)
    Create Camera Offset Attribute=true
    Create Color Attribute=true
    Write to Module Namespace=true
   )
   &FN_Light_Attributes(
    Color=$Particles.Color
    Create Color Attribute=true
    Create Exponent Attribute=true
    Create Position Attribute=true
    Create Radius Attribute=true
    Enable Particle Light Enable=true
    Exponent=2
    Light Enabled=true
    Position=&AddVectorToPosition(
     Position=$Particles.Position
     Vector=&RotateVectorByQuaternion(Vector To Rotate=<0 0 -10>, Quaternion=$Particles.SkeletalMeshLocation.SampledRotation)
    )
    Radius=8
    Scale Color=&Multiply_Float(
     A=&FN_Select_Float_from_Time_Of_Day(Morning=0.035, Day=0.15, Evening=0.035, Night=0.15)
     B=&FN_FrontEndFloatModifier(Gameplay=1, Front End=10)
    )
   )
   &FN_Light_Attributes(
    Color=$Particles.Color
    Create Color Attribute=true
    Create Exponent Attribute=true
    Create Position Attribute=true
    Create Radius Attribute=true
    Enable Particle Light Enable=true
    Exponent=2
    Light Enabled=true
    Position=&AddVectorToPosition(
     Position=$Particles.Position
     Vector=&RotateVectorByQuaternion(Vector To Rotate=<0 0 -6>, Quaternion=$Particles.SkeletalMeshLocation.SampledRotation)
    )
    Radius=8
    Scale Color=&Multiply_Float(
     A=&FN_Select_Float_from_Time_Of_Day(Morning=0.035, Day=0.15, Evening=0.035, Night=0.15)
     B=&FN_FrontEndFloatModifier(Gameplay=1, Front End=5)
    )
   ) {module=FN_Light_Attributes001}
   &CameraOffset(Camera Offset Amount=5)
  ]
  rendererClasses [
   &NiagaraMeshRendererProperties
   &NiagaraSpriteRendererProperties
   &NiagaraLightRendererProperties
  ]
  renderers [
   {rendererClass=&NiagaraMeshRendererProperties, rendererIndex=0}
   {rendererClass=&NiagaraSpriteRendererProperties, rendererIndex=1}
   {rendererClass=&NiagaraLightRendererProperties, rendererIndex=2}
   {rendererClass=&NiagaraLightRendererProperties, rendererIndex=3}
  ]
 }}
summary {
 systemName=NS_Backpack_DistantEchoPro_Light
 emitters [{
   emitterName=LightCone_FromHand
   rendererClasses [
    &NiagaraMeshRendererProperties
    &NiagaraSpriteRendererProperties
    &NiagaraLightRendererProperties
   ]
  }]
 userVariables [
  User.Alpha {type=NiagaraFloat, defaultValue=0, description=""}
  User.bIsFrontEnd {type=NiagaraBool}
  User.bIsFrontendPreview {type=NiagaraBool}
  User.bIsLargeBodyPawn {type=NiagaraBool}
  User.ConeAngle {type=NiagaraFloat, defaultValue=1, description=""}
  User.ConeLength {type=NiagaraFloat, defaultValue=10, description=""}
  User.ConeStartRadius {type=NiagaraFloat, defaultValue=0.75, description=""}
  User.SkeletalMesh {
   type=NiagaraDataInterfaceSkeletalMesh
   defaultValue {
    struct=NiagaraExt_VariableValue_DataInterface
    value {dataInterface=&self:NiagaraDataInterfaceSkeletalMesh_1}
   }
  }
 ]
}
