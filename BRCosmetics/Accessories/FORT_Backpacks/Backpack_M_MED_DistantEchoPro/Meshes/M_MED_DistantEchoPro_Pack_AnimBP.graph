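# Animation Blueprint for the DistantEchoPro backpack cosmetic, playing the pack's idle animations.
# A Random Sequence Player favors the main idle over an occasional idle break, then a Modify Bone node replaces the root bone's scale to shrink the pack.
# An "Animation Testing" state machine alternating the same two clips is left unconnected to the output pose.
# The event graph holds no logic.
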AnimBlueprint /BRCosmetics/Accessories/FORT_Backpacks/Backpack_M_MED_DistantEchoPro/Meshes/M_MED_DistantEchoPro_Pack_AnimBP
target_skeleton=&/M_MED_DistantEchoPro_Pack_Skeleton
graphs [
 AnimGraph {nodes [
   Input Pose()
   n1 = Random Sequence Player!() -> (Pose:Pose Link Structure) {
    is=title AnimGraphNode_RandomPlayer
    entries [
     {
      chanceToPlay=0.8
      maxLoopCount=0
      maxPlayRate=1
      minLoopCount=0
      minPlayRate=1
      sequence=/BRCosmetics/Accessories/FORT_Backpacks/Backpack_M_MED_DistantEchoPro/Animation/DistantEchoPro_Idle_PROP
      blendIn {blendOption=Linear, blendTime=0.2, customCurve=null}
     }
     {
      chanceToPlay=0.2
      maxLoopCount=0
      maxPlayRate=1
      minLoopCount=0
      minPlayRate=1
      sequence=/BRCosmetics/Accessories/FORT_Backpacks/Backpack_M_MED_DistantEchoPro/Animation/DistantEchoPro_Idle_Break_PROP
      blendIn {blendOption=Linear, blendTime=0.2, customCurve=null}
     }
    ]
   }
   n2 = "Transform (Modify) Bone - Bone: root"(ComponentPose=Local To Component(LocalPose=@n1), Translation="0,0,0", Rotation="0,0,0", Scale="0.65,0.65,0.65") {scaleMode=BMM_Replace, alphaBoolBlend {blendOption=Linear}, boneToModify {boneName=root}, scale=<0.65 0.65 0.65>}
   Output Pose(Result=Component To Local(ComponentPose=@n2)) {layerGroup=None}
   AnimGraphNode_StateMachine_0 = Animation Testing() {is=title AnimGraphNode_StateMachine}
  ]}
 Animation Testing {
  entry=State
  path=AnimGraph.AnimGraphNode_StateMachine_0.Animation Testing
  ids=id
  nodes [
   AnimStateNode_0 = State() {is=title AnimStateNode}
   AnimStateNode_1 = State_1() {is=title AnimStateNode}
   AnimStateTransitionNode_0 = AnimStateTransitionNode() {bAutomaticRuleBasedOnSequencePlayerInState=true, crossfadeDuration=0}
   AnimStateTransitionNode_1 = AnimStateTransitionNode() {bAutomaticRuleBasedOnSequencePlayerInState=true, crossfadeDuration=0}
  ]
  links [
   {
    dst=State_1
    rule=AnimGraph.AnimGraphNode_StateMachine_0.Animation Testing.AnimStateTransitionNode_0.Transition
    src=State
    via=AnimStateTransitionNode_0
   }
   {
    dst=State
    rule=AnimGraph.AnimGraphNode_StateMachine_0.Animation Testing.AnimStateTransitionNode_1.Transition
    src=State_1
    via=AnimStateTransitionNode_1
   }
  ]
 }
 State {
  path=AnimGraph.AnimGraphNode_StateMachine_0.Animation Testing.AnimStateNode_0.State
  nodes [
   n1 = Sequence Player 'DistantEchoPro_Idle_PROP'() {bLoopAnimation=false, sequence=&DistantEchoPro_Idle_PROP}
   Output Animation Pose(Result=@n1)
  ]
 }
 State_1 {
  path=AnimGraph.AnimGraphNode_StateMachine_0.Animation Testing.AnimStateNode_1.State_1
  nodes [
   n1 = Sequence Player 'DistantEchoPro_Idle_Break_PROP'() {
    bLoopAnimation=false
    sequence=&DistantEchoPro_Idle_Break_PROP
   }
   Output Animation Pose(Result=@n1)
  ]
 }
 Transition {path=AnimGraph.AnimGraphNode_StateMachine_0.Animation Testing.AnimStateTransitionNode_0.Transition, nodes [Result()]}
 Transition {path=AnimGraph.AnimGraphNode_StateMachine_0.Animation Testing.AnimStateTransitionNode_1.Transition, nodes [Result()]}
 EventGraph {nodes [
   TryGetPawnOwner()
   Event BlueprintUpdateAnimation()
  ]}
]
variables []
